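# Sound cue for the airstrike gadget's explosion impact.
# A random node picks one of three equally weighted hit waves, and a modulator varies its pitch, with a raised upper limit, to keep repeats from sounding identical.
# It uses airstrike attenuation and the large explosives sound class, primes on load, and overrides concurrency to cap simultaneous plays at a few, stopping the oldest.
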SoundCue /Game/Sounds/Fort_Gadgets/Airstrike/airstrike_explosion_Cue
root=Modulator_0
nodes [
 {id=Modulator_0, children [Random_1], properties {pitchMax=1.3}}
 {id=Random_1, children [WavePlayer_0,WavePlayer_1,WavePlayer_2], properties {weights [1,1,1]}}
 {id=WavePlayer_0, properties {soundWaveAssetPtr=&/Airstrik_Hit_01}}
 {id=WavePlayer_1, properties {soundWaveAssetPtr=&/Airstrik_Hit_02}}
 {id=WavePlayer_2, properties {soundWaveAssetPtr=&/Airstrik_Hit_03}}
]
properties {
 attenuationSettings=/Game/Sounds/Attenuation/Fort_AirStrike_Att
 bOverrideConcurrency=true
 bPrimeOnLoad=true
 maxDistance=9000
 soundClassObject=/Game/Sounds/FortSoundClassesAndModes/Fort_Explosives_Large_PSM
 volumeMultiplier=0.4953
 concurrencyOverrides {maxCount=3, resolutionRule=StopOldest}
}
