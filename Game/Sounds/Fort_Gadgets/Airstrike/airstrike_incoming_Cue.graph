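# Sound cue for the airstrike gadget's incoming-fire sound, using airstrike attenuation.
# A quality-level root selects between an equally weighted random pick of four incoming-fire waves and a three-wave subset, leaving its other two slots empty.
# Both branches pass through a pitch modulator for variation.
# A concurrency override allows one instance and blocks new ones while it plays, and the cue primes on load.
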SoundCue /Game/Sounds/Fort_Gadgets/Airstrike/airstrike_incoming_Cue
root=QualityLevel_0
nodes [
 {id=Modulator_1, children [Random_1], properties {pitchMax=1.3}}
 {id=Modulator_2, children [Random_2], properties {pitchMax=1.3}}
 {id=QualityLevel_0, children [Modulator_1,Modulator_2,null,null]}
 {id=Random_1, children [WavePlayer_4,WavePlayer_5,WavePlayer_6,WavePlayer_7], properties {weights [1,1,1,1]}}
 {id=Random_2, children [WavePlayer_8, WavePlayer_10, WavePlayer_11], properties {weights [1,1,1]}}
 {id=WavePlayer_10, properties {soundWaveAssetPtr=&/incoming_fire_03}}
 {id=WavePlayer_11, properties {soundWaveAssetPtr=&/incoming_fire_04}}
 {id=WavePlayer_4, properties {soundWaveAssetPtr=&/incoming_fire_01}}
 {id=WavePlayer_5, properties {soundWaveAssetPtr=&/incoming_fire_02}}
 {id=WavePlayer_6, properties {soundWaveAssetPtr=&incoming_fire_03}}
 {id=WavePlayer_7, properties {soundWaveAssetPtr=&incoming_fire_04}}
 {id=WavePlayer_8, properties {soundWaveAssetPtr=&incoming_fire_01}}
]
properties {
 attenuationSettings=/Game/Sounds/Attenuation/Fort_AirStrike_Att
 bOverrideConcurrency=true
 bPrimeOnLoad=true
 maxDistance=9000
 soundClassObject=&Fort_WeaponRanged_Fire_3P_PSM
 volumeMultiplier=0.8
 concurrencyOverrides {maxCount=1, resolutionRule=PreventNew}
}
