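# Sound cue for the airstrike gadget, which by its name plays at the incoming strike's placement.
# A single Wave Player loops airstrike_incoming_place_01 directly into the root, with no mixing or randomization.
# It uses the gadget teleporter attenuation asset and the Fort_Gadgets_StW sound class, with a slight volume boost and priming on load.
# A concurrency override rejects new instances at the limit rather than stopping playing ones.
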SoundCue /Game/Sounds/Fort_Gadgets/Airstrike/airstrike_incoming_place_Cue
root=WavePlayer_2
nodes [{id=WavePlayer_2, properties {bLooping=true, soundWaveAssetPtr=&/airstrike_incoming_place_01}}]
properties {
 attenuationSettings=/Game/Sounds/Attenuation/Fort_GadgetTeleporter_Att
 bOverrideConcurrency=true
 bPrimeOnLoad=true
 maxDistance=4500
 soundClassObject=/Game/Sounds/FortSoundClassesAndModes/Fort_Gadgets_StW
 volumeMultiplier=1.2
 concurrencyOverrides {resolutionRule=PreventNew}
}
