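# Sound cue for the airstrike gadget that plays a stereo airstrike call sound.
# A single wave player feeds the output directly at boosted volume, primed on load and routed to a melee weapon sound class.
# Overridden attenuation turns off distance volume falloff, superseding the assigned airstrike attenuation asset.
# Overridden concurrency allows only one instance and blocks new plays while it is active.
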SoundCue /Game/Sounds/Fort_Gadgets/Airstrike/airstrike_stereo_all_Cue
root=WavePlayer_1
nodes [{id=WavePlayer_1, properties {soundWaveAssetPtr=&/Airstrik_Call_01}}]
properties {
 attenuationSettings=/Game/Sounds/Attenuation/Fort_AirStrike_Att
 bOverrideAttenuation=true
 bOverrideConcurrency=true
 bPrimeOnLoad=true
 maxDistance=2097152
 soundClassObject=&Fort_WeaponMelee_PSM
 volumeMultiplier=2.2
 attenuationOverrides {bAttenuate=false, distanceAlgorithm=Logarithmic, falloffDistance=8000, stereoSpread=5000, attenuationShapeExtents {x=4000}}
 concurrencyOverrides {maxCount=1, resolutionRule=PreventNew}
}
